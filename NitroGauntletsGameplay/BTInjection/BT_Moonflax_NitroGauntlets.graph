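# Behavior tree for Nitro Gauntlets AI, built on the UseItem blackboard, that closes on the target actor and uses the injected item.
# A root Selector runs a Sequence that moves toward the tracked target, allowing partial paths, then uses the item with alternate fire driven by a blackboard key.
# If that Sequence fails, it falls back to a blackboard-timed wait.
# A SetInTacticalAction service runs on the root throughout.
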BehaviorTree /NitroGauntletsGameplay/BTInjection/BT_Moonflax_NitroGauntlets
blackboard=/AICore/Humanoids/BehaviorTrees/Blackboards/BB_UseItem
root {
 name=Selector
 class=BTComposite_Selector
 children [
  Sequence {
   class=BTComposite_Sequence
   children [
    Move To {
     class=BTTask_MoveTo
     children []
     decorators []
     properties {
      bIgnoreRestartSelf=false
      nodeName=Move To
      acceptableRadius {defaultValue=500, key=None}
      bAllowPartialPath {defaultValue=true, key=None}
      bAllowStrafe {defaultValue=false, key=None}
      bProjectGoalLocation {defaultValue=true, key=None}
      bReachTestIncludesAgentRadius {defaultValue=true, key=None}
      bReachTestIncludesGoalRadius {defaultValue=true, key=None}
      bRequireNavigableEndLocation {defaultValue=true, key=None}
      bStartFromPreviousPath {defaultValue=false, key=None}
      bTrackMovingGoal {defaultValue=true, key=None}
      blackboardKey {bNoneIsAllowedValue=false, selectedKeyName=AIEvaluator_Global_TargetActor}
      filterClass {baseClass=&NavigationQueryFilter, defaultValue=null, key=None}
      observedBlackboardValueTolerance {defaultValue=100, key=None}
     }
     services []
    }
    Use Item {
     class=FortAthenaBTTask_UseItem
     decorators []
     properties {
      actionObjectKeyName=AIEvaluator_Injection_Item
      executionStatusKeyName=None
      nodeName=Use Item
      finishAfterItemUse {defaultValue=false}
      inputHoldDuration {key=None}
      maximumWaitTimeBetweenUses {defaultValue=0}
      resetActionObjectKey {defaultValue=true}
      useAlternateMode {key=Item_UseAlternateFire}
     }
    }
   ]
   decorators []
   properties {children [
     {childTask=&self:BTTask_MoveTo_0}
     {childTask=&self:FortAthenaBTTask_UseItem_0}
    ]}
  }
  Wait Blackboard Time {
   class=BTTask_WaitBlackboardTime
   children []
   decorators []
   properties {
    bIgnoreRestartSelf=false
    nodeName=Wait Blackboard Time
    blackboardKey {bNoneIsAllowedValue=false, selectedKeyName=Item_WaitTimeAfterUse}
    randomDeviation {defaultValue=0, key=None}
    waitTime {defaultValue=5, key=Item_WaitTimeAfterUse}
   }
   services []
  }
 ]
 decorators []
 properties {
  children [
   {childComposite=&self:BTComposite_Sequence_1, childTask=null, decorators []}
   {childComposite=null, childTask=&self:BTTask_WaitBlackboardTime_0, decorators []}
  ]
  services [&self:FortAthenaBTService_DontForgetCurrentThreat_0]
 }
 services [FortAthenaBTService_DontForgetCurrentThreat_0 {
   class=FortAthenaBTService_SetInTacticalAction
   properties {shouldSetInTacticalAction {key=None}, tacticalActionIdentifierTag {defaultValue {_structType=&GameplayTag}}}
  }]
}
