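; Bot AI evaluator that decides whether a Fortnite AI bot can use Nitro Gauntlets.
; On enter it caches the owning controller's blackboard.
; Starting and ongoing conditions both call one shared check on the bot's pawn.
; It requires that the pawn is not surface swimming, the blackboard reports line of sight on a threat, and the pawn lacks the Nitro Gauntlets prevent tag.
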(event EventOnEnter
  (params (OwnerComp :type Behavior%20Tree%20Component%20Object%20Reference))
  (Variables|Default|SetBlackboardComponent (AI|GetBlackboard (Components|GetOwner OwnerComp))))

(fn CanUseNitroGauntlets
  (params (OwnerComp :type Behavior%20Tree%20Component%20Object%20Reference))
  (Utilities|IsValid (Variables|Default|GetBlackboardComponent)
    (:Is%20Valid
      (bind _asfort_athena_aibot_controller (Utilities|Casting|CastToFortAthenaAIBotController (Components|GetOwner OwnerComp))
        (:then
          (Utilities|IsValid (Variables|/Script/FortniteAI.FortAthenaAIBotController|GetPlayerBotPawn _asfort_athena_aibot_controller)
            (:Is%20Valid
              (return (Math|Boolean|ANDBoolean (not (WaterBody|IsSurfaceSwimming (Variables|/Script/FortniteAI.FortAthenaAIBotController|GetPlayerBotPawn _asfort_athena_aibot_controller))) (AI|Components|Blackboard|GetValueasBool (Variables|Default|GetBlackboardComponent) (Variables|Default|GetAIEvaluator_Global_HasLoSOnThreat)) (not (GameplayTags|HasTag (GameplayTags|GetOwnedGameplayTags (Variables|/Script/FortniteAI.FortAthenaAIBotController|GetPlayerBotPawn _asfort_athena_aibot_controller) :!class /Script/GameplayTags.GameplayTagAssetInterface) "(TagName=\"Ability.Moonflax.Nitrogauntlets.Prevent\")" true)))))
            (:Is%20Not%20Valid
              _)))
        (:CastFailed
          _))
      (return false))
    (:Is%20Not%20Valid
      _)))

(fn EvaluateOngoingConditions
  (params (OwnerComp :type Behavior%20Tree%20Component%20Object%20Reference))
  (bind _isusable (CallFunction|CanUseNitroGauntlets OwnerComp))
  (return _isusable))

(fn EvaluateStartingConditions
  (params (OwnerComp :type Behavior%20Tree%20Component%20Object%20Reference))
  (bind _isusable (CallFunction|CanUseNitroGauntlets OwnerComp))
  (return _isusable))
